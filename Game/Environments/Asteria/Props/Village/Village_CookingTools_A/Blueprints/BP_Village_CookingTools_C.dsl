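; Construction-script Blueprint for a village cooking tools prop in the Asteria environment, with an optional steam effect.
; The construction script calls the parent's script, then sets auto-activate on the NS_Village_CookingTools_Steam component from the Set Steam Active variable.
; BeginPlay, overlap and Tick events are empty, so the steam toggle is the only logic.
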(event EventBeginPlay)

(event Collision|EventActorBeginOverlap
  (params (OtherActor :type Actor%20Object%20Reference)))

(event EventTick
  (params (DeltaSeconds :type Float%20%28single-precision%29)))

(fn ConstructionScript
  (|CallParentFunction)
  (Components|Activation|SetAutoActivate (Variables|Default|GetNS_Village_CookingTools_Steam) (Variables|Default|GetSet%20Steam%20Active)))
