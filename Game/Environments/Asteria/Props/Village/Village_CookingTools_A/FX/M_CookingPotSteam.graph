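# Translucent, two-sided steam material for the Asteria village cooking pot, built for Niagara mesh particles.
# Color follows time-of-day tints, darkens by camera-to-sun direction and passes through emissive adjustments, with optional particle color and alpha.
# Noise-distorted, panning smoke forms the opacity, faded at UV ends and grazing angles, with platform-switched depth fade.
# Three summed time-driven sines along V form a normal-direction wobble, apparently for vertex offset.
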Material /Game/Environments/Asteria/Props/Village/Village_CookingTools_A/FX/M_CookingPotSteam
graph {
 properties {
  bUsedWithNiagaraMeshParticles=true
  blendMode=BLEND_Translucent
  pixelDepthOffsetMode=PDOM_Legacy
  twoSided=true
  usedShadingModels=Default Lit
 }
 nodes [
  n1 = ParticleColor()
  n2 = Constant3Vector() {constant=<rgba 0.3961 0.4 0.4 0>}
  n3 = Multiply(
   A=MaterialFunctionCall(
    Color_Morning=VectorParameter() {parameterName=Color Morning Multiplier, defaultValue=<rgba 0.9896 0.268 0.4929 0>}.RGB
    Color_Day=VectorParameter() {parameterName=Color Day Multiplier, defaultValue=<rgba 1 1 1 0>}.RGB
    Color_Evening=VectorParameter() {parameterName=Color Evening Multiplier, defaultValue=<rgba 0.599 0.2901 0.3358 0>}.RGB
    Color_Night=VectorParameter() {parameterName=Color Night Multiplier, defaultValue=<rgba 0.3498 0.4864 0.6458 0>}.RGB
   ) {materialFunction=&Color_Time_of_Day}.Result
   B=@n2
  )
  n4 = ScalarParameter() {defaultValue=0.7, parameterName=ShadowMultiplier}
  n5 = CollectionParameter() {
   collection=&FortniteMaterialParameters
   parameterName=SunlightDirectionalVector
  }
  n6 = DotProduct(A=CameraVectorWS(), B=ComponentMask(None=@n5) {b=true, g=true, r=true})
  n7 = MaterialFunctionCall(Emissive=LinearInterpolate(A=Multiply(A=@n3, B=@n4), B=@n3, Alpha=Saturate(None=@n6))) {materialFunction=&MF_FX_EmissiveAdjustments}
  n8 = ScalarParameter() {parameterName=UseParticleColorAlpha}
  LinearInterpolate(A=@n7.Result, B=Multiply(A=@n7.Result, B=@n1.RGB), Alpha=@n8)
  n9 = Multiply(
   A=TextureCoordinate()
   B=AppendVector(A=ScalarParameter() {defaultValue=7, parameterName=X tile}, B=ScalarParameter() {defaultValue=1, parameterName=Y tile})
  )
  n10 = ComponentMask(None=@n9) {r=true}
  n11 = TextureSample(UVs=Panner(Coordinate=TextureCoordinate() {uTiling=4}) {speedX=-0.5, speedY=1}) {samplerType=SAMPLERTYPE_Grayscale, texture=/FortWater/Textures/LowResBlurredNoise}
  n12 = ComponentMask(None=@n9) {g=true}
  n13 = Multiply(A=Multiply(A=@n11.R) {constB=0.25}, B=OneMinus(None=@n12))
  n14 = Panner(Coordinate=AppendVector(A=Add(A=@n10, B=@n13), B=Power(Base=@n12) {constExponent=0.5})) {speedX=0.5, speedY=1.5}
  n15 = ComponentMask(None=TextureCoordinate()) {g=true}
  n16 = OneMinus(None=@n15)
  n17 = Multiply(
   A=Multiply(A=TextureSample(UVs=@n14) {texture=&T_Inky_Smoke_Tile}.R, B=@n11.R)
   B=Multiply(A=@n15, B=SmoothStep(Value=@n16) {constMax=0.5})
  )
  n18 = CameraVectorWS()
  n19 = VertexNormalWS()
  n20 = Saturate(None=Multiply(A=Multiply(A=@n17) {constB=1.25}, B=DotProduct(A=@n18, B=@n19)))
  n21 = Multiply(A=@n20) {constB=0.5}
  n22 = QualitySwitch(Default=@n21, Low=@n20, High=@n21, Medium=@n20, Epic=@n21)
  n23 = DepthFade(Opacity=@n22, FadeDistance=ScalarParameter() {parameterName=Depth Fade}) {fadeDistanceDefault=0}
  n24 = MaterialFunctionCall("Xb1,PS4"=@n23, "PC-High&&Epic,XSX,PS5"=@n23, "PC-Perf,Android"=@n22, Switch=@n22, "PC-Low(NotPerf)&&Medium"=@n23) {materialFunction=&PlatformSwitch_1D}
  LinearInterpolate(A=@n24.Result, B=Multiply(A=@n1.A, B=@n24.Result), Alpha=@n8)
  n25 = VertexNormalWS()
  n26 = OneMinus(None=Time())
  n27 = Sine(None=Add(A=@n16, B=Multiply(A=@n26) {constB=0.75})) {period=0.5}
  n28 = Sine(None=Add(A=@n16, B=Multiply(A=@n26))) {period=0.5}
  n29 = Add(A=@n16, B=Multiply(A=@n26) {constB=0.5})
  n30 = ConstantBiasScale(None=Add(A=Add(A=@n27, B=@n28), B=Sine(None=@n29)))
  Multiply(A=Multiply(A=Multiply(A=@n25, B=@n16), B=Multiply(A=@n30) {constB=5}), B=SmoothStep(Value=@n16) {constMin=0.1})
  Reroute(None=@n2)
  TextureCoordinate() {uTiling=7}
 ]
}
